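# Unlit, masked material for the LuckyOnyx01 vehicle wheel, with glowing, animated lava-style color effects.
# MF_WheelParent_Simple supplies base attributes and an RGB mask that splits the color into base and masked regions.
# Emissive blends brightened, Fresnel-lit color with panning caustic patterns, using sine-pulse and camera-facing terms; static switch LOD2345 selects one layer's pan axis.
# The result passes through vehicle damage and de-rez transition functions.
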Material /VehicleCosmetics/Wheels/LuckyOnyx01/Materials/M_Wheel_LuckyOnyx01
graph {
 properties {
  bAutomaticallySetUsageInEditor=false
  bUseMaterialAttributes=true
  bUsedWithMeshDeformer=true
  bUsedWithSkeletalMesh=true
  blendMode=BLEND_Masked
  pixelDepthOffsetMode=PDOM_Legacy
  previewMesh=/VehicleCosmetics/Wheels/LuckyOnyx01/Meshes/SK_Wheel_LuckyOnyx01
  shadingModel=MSM_Unlit
  usedShadingModels=Unlit
 }
 nodes [
  n1 = MaterialFunctionCall() {materialFunction=&MF_WheelParent_Simple}
  n2 = GetMaterialAttributes(None=@n1.MaterialAttributes) {attributeGetTypes [#1, #2]}
  n3 = NamedRerouteUsage()
  n4 = NamedRerouteUsage()
  n5 = LinearInterpolate(A=@n3, B=Multiply(A=@n3, B=ScalarParameter() {defaultValue=3, group=5_Special Effects, parameterName=ColorBaseBrightness}), Alpha=@n4)
  n6 = Fresnel(Normal=NamedRerouteUsage()) {exponent=3}
  n7 = Multiply(
   A=Multiply(A=@n3, B=Multiply(A=@n4, B=@n6))
   B=Multiply(A=ScalarParameter() {defaultValue=6, group=5_Special Effects, parameterName=FresnelIntensity}) {constB=10}
  )
  n8 = NamedRerouteUsage()
  n9 = NamedRerouteUsage()
  n10 = Multiply(A=TextureCoordinate(), B=ScalarParameter() {defaultValue=4, group=5_Special Effects, parameterName=LavaTile})
  n11 = Multiply(A=ScalarParameter() {defaultValue=1, group=5_Special Effects, parameterName=PannerSpeed}, B=Time())
  n12 = LinearInterpolate(
   A=LinearInterpolate(A=Multiply(A=@n8, B=@n9)) {constB=0, desc=DarkerColor}
   B=@n9
   Alpha=TextureSampleParameter2D(UVs=Panner(Coordinate=@n10, Time=@n11) {speedX=0.15, speedY=0.15}) {group=5_Special Effects, parameterName=PatternMask, texture=/VehicleCosmetics/Wheels/LuckyOnyx01/Textures/Caustics_D}.R
  )
  n13 = NamedRerouteUsage()
  n14 = NamedRerouteUsage()
  n15 = Panner(Coordinate=@n10, Time=@n11) {speedX=0.15}
  n16 = Panner(Coordinate=@n10, Time=@n11) {speedY=0.15}
  n17 = LinearInterpolate(
   A=LinearInterpolate(A=Multiply(A=@n13, B=@n14)) {constB=0, desc=DarkerColor}
   B=@n14
   Alpha=StaticSwitchParameter(
    True=TextureSampleParameter2D(UVs=@n15) {group=5_Special Effects, parameterName=PatternMask, texture=&Caustics_D}.R
    False=TextureSampleParameter2D(UVs=@n16) {group=5_Special Effects, parameterName=PatternMask, texture=&Caustics_D}.R
   ) {group=5_Special Effects, parameterName=LOD2345}
  )
  n18 = LinearInterpolate(
   A=LinearInterpolate(A=Add(A=@n5, B=@n7) {constB=0}, B=NamedRerouteUsage(), Alpha=NamedRerouteUsage())
   B=Multiply(A=Add(A=@n12, B=@n17), B=NamedRerouteUsage())
   Alpha=Add(A=NamedRerouteUsage(), B=NamedRerouteUsage())
  )
  MaterialFunctionCall(In=MaterialFunctionCall(Material=SetMaterialAttributes(MaterialAttributes=@n2.MaterialAttributes, Emissive Color=@n18) {attributeSetTypes [#3]}) {materialFunction=&MF_BaseVehicle_Damage}.Result) {materialFunction=&MF_VehicleDeRez_Transition}
  NamedRerouteDeclaration(None=@n2.Normal) {name=normal_output, nodeColor=<rgba 1 0.1875 0 1>}
  n19 = MaterialFunctionCall(Color=@n1.RimRGBAMask) {materialFunction=&SplitComponents}
  NamedRerouteDeclaration(None=Multiply(A=@n2.Base Color, B=OneMinus(None=@n19.RGB))) {name=ColorBase, nodeColor=<rgba 0 0.1328 1 1>}
  n20 = Sine(None=Multiply(A=Time()) {constB=0.2})
  NamedRerouteDeclaration(None=Multiply(
    A=ConstantBiasScale(None=@n20) {bias=1.25}
    B=ScalarParameter() {defaultValue=1, group=5_Special Effects, parameterName=PulseEmissiveIntensity}
   )) {name=PulsingFX, nodeColor=<rgba 0 0.4141 1 1>}
  NamedRerouteDeclaration(None=Multiply(A=@n2.Base Color, B=NamedRerouteDeclaration(None=@n19.G) {name=GMask, nodeColor=<rgba 0 1 0.9062 1>})) {name=ColorB, nodeColor=<rgba 0.4531 1 0 1>}
  n21 = Constant3Vector() {constant=<rgba 0 0 1 0>}
  n22 = OneMinus(None=DotProduct(A=Transform(None=@n21), B=CameraVectorWS()))
  NamedRerouteDeclaration(None=Saturate(None=ConstantBiasScale(None=@n22) {bias=-0.5, scale=5})) {name=Camera_Fresnel, nodeColor=<rgba 0 1 0.3438 1>}
  NamedRerouteDeclaration(None=Multiply(A=NamedRerouteDeclaration(None=@n19.B) {name=BMask, nodeColor=<rgba 1 0 0.6328 1>}, B=@n2.Base Color)) {name=ColorC, nodeColor=<rgba 0 0.9531 1 1>}
  NamedRerouteDeclaration(None=Multiply(A=NamedRerouteDeclaration(None=@n19.R) {name=RMask, nodeColor=<rgba 1 0 0.2578 1>}, B=@n2.Base Color)) {name=ColorA, nodeColor=<rgba 1 0.1875 0 1>}
 ]
}
